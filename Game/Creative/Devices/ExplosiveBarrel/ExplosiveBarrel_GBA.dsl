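; Fortnite Creative explosive barrel device that detonates during a minigame on death, team-filtered proximity, timer or trigger event.
; Detonation queues a throttled async sphere overlap, sends a damage gameplay event per damageable hit, broadcasts On Exploded, then hides the barrel.
; Minigame state and reset events restore health, collision, invulnerability and fuse.
; Options drive mesh, team visibility, health bar and range preview; net dormancy is flushed on replicated changes.
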(event MinigameStateChangedDelegate(FortMinigameLogic)
  (params (Minigame :type Fort%20Minigame%20Object%20Reference) (NewMinigameState :type EFortMinigameState%20Enum :default "PreGame"))
  ;[0.25 0.25 0.25] Handle Visibility and collision when MinigameStateChanged
  (Utilities|FlowControl|Sequence
    (:then_0
      (CallFunction|UpdateVisualizationState))
    (:then_1
      (if (Utilities|Enum|Equal(Enum) NewMinigameState "PostGameReset")
        (CallFunction|SetCollisionForcedDuringGame "NewEnumerator0")
        ;[1 1 1] Cancel detonation timer when minigame ends
        (Utilities|Time|ClearandInvalidateTimerbyHandle (Variables|Default|GetDetonationTimer))
        (CallFunction|SetInvulneravilitybyMinigameState)
        (CallFunction|RestoretoMaxHealth)
        (CallFunction|UpdateBarrelMeshCollision)
        (Variables|Default|SetExplosiveFuseLit false)))))

(event MinigameEndedDelegate(FortMinigameLogic)
  ;[1 1 1] Cancel detonation timer when minigame ends
  (Utilities|Time|ClearandInvalidateTimerbyHandle (Variables|Default|GetDetonationTimer))
  (CallFunction|SetInvulneravilitybyMinigameState)
  (CallFunction|RestoretoMaxHealth)
  (CallFunction|UpdateBarrelMeshCollision)
  (Variables|Default|SetExplosiveFuseLit false))

(event Custom|Reset%20Barrel%20State
  (if (not (Utilities|Time|IsTimerActivebyHandle (Variables|Default|GetResetTimerHandle)))
    (Utilities|FlowControl|Sequence
      (:then_0
        (CallFunction|SetCollisionForcedDuringGame "NewEnumerator0")
        ;[1 1 1] Cancel detonation timer when minigame ends
        (Utilities|Time|ClearandInvalidateTimerbyHandle (Variables|Default|GetDetonationTimer))
        (CallFunction|SetInvulneravilitybyMinigameState)
        (CallFunction|RestoretoMaxHealth)
        (CallFunction|UpdateBarrelMeshCollision)
        (Variables|Default|SetExplosiveFuseLit false))
      (:then_1
        (CallFunction|SetVisibilityForcedDuringGame "NewEnumerator0")
        (Visibility|UpdateVisibility (Variables|Default|GetVisibleInGameComponent))
        (bind _returnvalue (Utilities|Time|SetTimerbyEvent (delegate Custom|On%20Reset%20Timer%20Expired) 1.0))
        (Variables|Default|SetResetTimerHandle _returnvalue))
      (:then_2
        ;[1 1 1] Setup Detonation Timer
        (if (> (Variables|UserKnobs|GetTimerDetonateSetting) 0.0)
          (bind _returnvalue_2 (Utilities|Time|SetTimerbyEvent (delegate Custom|TimerDetonationEvent) (Variables|UserKnobs|GetTimerDetonateSetting)))
          (Variables|Default|SetDetonationTimer _returnvalue_2)
          (Variables|Default|SetExplosiveFuseLit true))))))

(event Custom|On%20Reset%20Timer%20Expired)

(event MinigameStartedDelegate(FortMinigameLogic)
  ;[1 1 1] Timer detonation.  Only works when a minigame is started.
  (CallFunction|SetInvulneravilitybyMinigameState)
  ;[1 1 1] Setup Detonation Timer
  (if (> (Variables|UserKnobs|GetTimerDetonateSetting) 0.0)
    (bind _returnvalue (Utilities|Time|SetTimerbyEvent (delegate Custom|TimerDetonationEvent) (Variables|UserKnobs|GetTimerDetonateSetting)))
    (Variables|Default|SetDetonationTimer _returnvalue)
    (Variables|Default|SetExplosiveFuseLit true)))

(event Custom|TimerDetonationEvent
  ;[1 1 1] Setup Detonation Timer
  (CallFunction|ExplodeBarrel))

(event EventBeginPlay
  ;[0 1 0.2776 0.25] Initialize settings
  (Visibility|SetVisibilityComponents (Variables|Default|GetVisibleInGameComponent) (Utilities|Array|MakeArray (Variables|Default|GetSM_MilitaryBarrel) (Variables|Default|GetBombSparkFX)) true)
  (CallFunction|SetInvulneravilitybyMinigameState)
  (Class|FortAsyncActionAuthorityMatchReady|MatchReadyAsync
    (:OnReady
      (CallFunction|SetMaxHealth)
      (Components|Sphere|SetSphereRadius (Variables|Default|GetProximityTrigger) (Math|Conversions|ToFloat(Integer) (Variables|UserKnobs|GetProximityDistanceSetting)) false) ; Sets size of prox trigger based on setting
      (CallFunction|UpdateHealthWidget)
      (CallFunction|UpdateBarrelMeshCollision)
      (CallFunction|UpdateVisibilityComponent)
      (Networking|SetNetDormancy "DORM_DormantAll"))) ; Ensures everything exists before getting the initial settings correct. \nWithout it the Set Max Health fails because the Ability system component doesn't exist when begin play fires.
)

(event Custom|Initialize%20Device
  ;[0 1 0.2776 0.25] Initialize settings
  (CallFunction|UpdateVisualMesh)
  (CallFunction|UpdateVisualizationState)
  (CallFunction|SetMaxHealth)
  (Components|Sphere|SetSphereRadius (Variables|Default|GetProximityTrigger) (Math|Conversions|ToFloat(Integer) (Variables|UserKnobs|GetProximityDistanceSetting)) false) ; Sets size of prox trigger based on setting
  (CallFunction|UpdateHealthWidget)
  (CallFunction|UpdateBarrelMeshCollision)
  (CallFunction|UpdateVisibilityComponent)
  (Networking|SetNetDormancy "DORM_DormantAll"))

(event EditorPreview|EventSetupForEditorPreview
  ;[0 1 0.2776 0.25] Initialize settings
  (CallFunction|InitializeDevice))

(event OnAnyPropertyChanged(ToyOptionsComponent)
  ;[0 1 0.2776 0.25] Initialize settings
  (CallFunction|InitializeDevice))

(event OnPropertiesLoaded(ToyOptionsComponent)
  ;[0 1 0.2776 0.25] Initialize settings
  (CallFunction|InitializeDevice))

(event Custom|PlayBarrelFX
  ;[1 1 1] FX when Barrel is killed with Damage
  (Networking|CanExecuteCosmeticEvents
    (:True
      (CallFunction|OnExplodeFX "0, 0, 0"))))

(event Custom|OnDelayedProximity
  ;[1 1 1] ProximityDelay
  (CallFunction|ExplodeBarrel))

(event OnComponentBeginOverlap(ProximityTrigger)
  (params (OverlappedComponent :type Primitive%20Component%20Object%20Reference) (OtherActor :type Actor%20Object%20Reference) (OtherComp :type Primitive%20Component%20Object%20Reference) (OtherBodyIndex :type Integer) (bFromSweep :type Boolean) (SweepResult :type Hit%20Result%20Structure%20%28by%20ref%29))
  ;[1 1 1] Proximity trigger explosion, only works during minigames
  (bind _isrunning (CallFunction|MinigameIsinProgress true))
  (if (and _isrunning (> (Variables|UserKnobs|GetProximityDistanceSetting) 0))
    (bind _shouldtrigger (CallFunction|ShouldTriggerProximityExplosion OtherActor))
    (if _shouldtrigger
      (if (> (Variables|UserKnobs|GetProximityDelayTime) 0.0)
        ;[1 1 1] ProximityDelay
        (if (not (Utilities|Time|IsTimerActivebyHandle (Variables|Default|GetDelayedProximityTH)))
          (bind _returnvalue (Utilities|Time|SetTimerbyEvent (EventDispatchers|CreateEvent :!event Custom|OnDelayedProximity) (Variables|UserKnobs|GetProximityDelayTime)))
          (Variables|Default|SetDelayedProximityTH _returnvalue)
          (Variables|Default|SetExplosiveFuseLit true))
        (else
          (CallFunction|ExplodeBarrel))))))

(event EventOnDeathPlayEffects
  (params (Damage :type Float%20%28single-precision%29) (DamageTags :type Gameplay%20Tag%20Container%20Structure%20%28by%20ref%29) (Momentum :type Vector :default "0, 0, 0") (HitInfo :type Hit%20Result%20Structure%20%28by%20ref%29) (InstigatedBy :type Fort%20Pawn%20Object%20Reference) (DamageCauser :type Actor%20Object%20Reference) (EffectContext :type Gameplay%20Effect%20Context%20Handle%20Structure))
  (bind _isrunning (CallFunction|MinigameIsinProgress false))
  (if _isrunning
    (Rendering|SetVisibility (Variables|Default|GetSM_MilitaryBarrel) false false)
    (Components|Activation|SetActive (Variables|Default|GetBombSparkFX) false false)))

(event EventOnDamageServer
  (params (Damage :type Float%20%28single-precision%29) (DamageTags :type Gameplay%20Tag%20Container%20Structure%20%28by%20ref%29) (Momentum :type Vector :default "0, 0, 0") (HitInfo :type Hit%20Result%20Structure%20%28by%20ref%29) (InstigatedBy :type Controller%20Object%20Reference) (DamageCauser :type Actor%20Object%20Reference) (EffectContext :type Gameplay%20Effect%20Context%20Handle%20Structure)))

(event Custom|Turn%20On%20Visibility%20Event
  (params (Instigator :type Controller%20Object%20Reference))
  ;[0.2393 0 1 0.25] Gameplay Functions
  (CallFunction|SetVisibilityForcedDuringGame "NewEnumerator1"))

(event EventOnDeathServer
  (params (Damage :type Float%20%28single-precision%29) (DamageTags :type Gameplay%20Tag%20Container%20Structure%20%28by%20ref%29) (Momentum :type Vector :default "0, 0, 0") (HitInfo :type Hit%20Result%20Structure%20%28by%20ref%29) (InstigatedBy :type Controller%20Object%20Reference) (DamageCauser :type Actor%20Object%20Reference) (EffectContext :type Gameplay%20Effect%20Context%20Handle%20Structure))
  ;[1 1 1] Blow up barrel
  (bind _isrunning (CallFunction|MinigameIsinProgress false))
  (if (and _isrunning (not (Variables|Replicated|GetHasExploded)))
    (bind _returnvalue (Utility|OverlapQueue|GetAsyncOverlapQueue self))
    (Utilities|FlowControl|Sequence
      (:then_0
        (Variables|Default|SetBarrelKiller InstigatedBy))
      (:then_1
        (Class|AsyncOverlapQueue|ThrottledAsyncSphereOverlapActors :self _returnvalue :Location (Transformation|GetWorldLocation (Variables|Default|GetProximityTrigger)) :Radius (Variables|UserKnobs|GetDamageRangeSetting) :bCheckLineOfSight (Math|Conversions|ToBoolean(Integer) (Variables|UserKnobs|GetBlastLineOfSight)) :ShouldCancelProcess (EventDispatchers|CreateEvent :!event ShouldCancelExplosion) :ProcessHitActorFunction (delegate Custom|DestroyHitActors) :CompletionFunction (delegate Custom|FinishedDamagingActors) :ObjectTypes "ObjectTypeQuery1" :TraceChannel "ECC_Visibility"))
      (:then_2
        (if (Variables|UserKnobs|GetbShowFX)
          (CallFunction|PlayBarrelFX)))
      (:then_3
        (Variables|Replicated|SetHasExploded true)
        (GameplayEvent|BroadcastEvent (Variables|UserOptions|Events|GetOn%20Exploded) 0 (Utilities|Struct|MakeGameplayEventDestroyed (Variables|Default|GetBarrelKiller)))
        (CallFunction|SetVisibilityForcedDuringGame "NewEnumerator2")
        (CallFunction|SetCollisionForcedDuringGame "NewEnumerator2")
        (Networking|FlushNetDormancy)
        (Utilities|Time|ClearandInvalidateTimerbyHandle (Variables|Default|GetDelayedProximityTH))
        (Variables|Default|SetExplosiveFuseLit false)))))

(event Custom|Turn%20Off%20Visibility%20Event
  (params (Instigator :type Controller%20Object%20Reference))
  ;[0.2393 0 1 0.25] Gameplay Functions
  (CallFunction|SetVisibilityForcedDuringGame "NewEnumerator2"))

(event Custom|Explode%20Barrel
  ;[1 1 1] Blow up barrel
  (bind _isrunning (CallFunction|MinigameIsinProgress false))
  (if (and _isrunning (not (Variables|Replicated|GetHasExploded)))
    (bind _returnvalue (Utility|OverlapQueue|GetAsyncOverlapQueue self))
    (Utilities|FlowControl|Sequence
      (:then_0
        (Variables|Default|SetBarrelKiller (stale EventOnDeathServer InstigatedBy)))
      (:then_1
        (Class|AsyncOverlapQueue|ThrottledAsyncSphereOverlapActors :self _returnvalue :Location (Transformation|GetWorldLocation (Variables|Default|GetProximityTrigger)) :Radius (Variables|UserKnobs|GetDamageRangeSetting) :bCheckLineOfSight (Math|Conversions|ToBoolean(Integer) (Variables|UserKnobs|GetBlastLineOfSight)) :ShouldCancelProcess (EventDispatchers|CreateEvent :!event ShouldCancelExplosion) :ProcessHitActorFunction (delegate Custom|DestroyHitActors) :CompletionFunction (delegate Custom|FinishedDamagingActors) :ObjectTypes "ObjectTypeQuery1" :TraceChannel "ECC_Visibility"))
      (:then_2
        (if (Variables|UserKnobs|GetbShowFX)
          (CallFunction|PlayBarrelFX)))
      (:then_3
        (Variables|Replicated|SetHasExploded true)
        (GameplayEvent|BroadcastEvent (Variables|UserOptions|Events|GetOn%20Exploded) 0 (Utilities|Struct|MakeGameplayEventDestroyed (Variables|Default|GetBarrelKiller)))
        (CallFunction|SetVisibilityForcedDuringGame "NewEnumerator2")
        (CallFunction|SetCollisionForcedDuringGame "NewEnumerator2")
        (Networking|FlushNetDormancy)
        (Utilities|Time|ClearandInvalidateTimerbyHandle (Variables|Default|GetDelayedProximityTH))
        (Variables|Default|SetExplosiveFuseLit false)))))

(event Custom|Explode%20Event
  (params (Instigator :type Controller%20Object%20Reference))
  ;[0.2393 0 1 0.25] Gameplay Functions
  (if (not (Variables|Replicated|GetHasExploded))
    (bind _isrunning (CallFunction|MinigameIsinProgress false)) ; only explode if the game is currently running.
    (if _isrunning
      (Variables|Default|SetTriggerBarrelKiller Instigator)
      (CallFunction|ExplodeBarrel))))

(event Custom|DestroyHitActors
  (params (HitActor :type Actor%20Object%20Reference))
  ;[1 1 1] Blow up barrel
  (CallFunction|BlowUpBarrel HitActor))

(event Custom|FinishedDamagingActors
  (params (WasCancelled :type Boolean)))

(event Custom|Reset%20Event
  (params (Instigator :type Controller%20Object%20Reference))
  ;[0.2393 0 1 0.25] Gameplay Functions
  (CallFunction|ResetBarrelState))

(event OnVisibilityChanged(VisibleInGameComponent)
  (params (Visible :type Boolean))
  (CallFunction|UpdateBarrelMeshCollision))

(event Custom|UpdateVisualMesh
  (bind _bombsparkfx (Variables|Default|GetBombSparkFX))
  (bind _sm_militarybarrel (Variables|Default|GetSM_MilitaryBarrel))
  ;[1 1 1] Set the Visual Mesh of the Device
  (bind _loaded_object (Creative|Loading|AsyncLoadEditorSafe (Variables|UserOptions|Advanced|GetDeviceMesh))
    (:Loaded
      (bind _asstatic_mesh (Utilities|Casting|CastToStaticMesh _loaded_object))
      (Components|StaticMesh|SetStaticMesh _sm_militarybarrel _asstatic_mesh)
      (Rendering|Material|SetMaterial _sm_militarybarrel 0 (StaticMesh|GetMaterial _asstatic_mesh 0))
      (if (== (Variables|UserOptions|Advanced|GetDeviceMesh) (Variables|Default|GetBombMeshReference))
        (bind _loaded_object_2 (Creative|Loading|AsyncLoadEditorSafe (Variables|Default|GetFuseFX))
          (:Loaded
            (bind _asniagara_system (Utilities|Casting|CastToNiagaraSystem _loaded_object_2))
            (Niagara|SetNiagaraSystemAsset _bombsparkfx _asniagara_system true)))
        (else
          (Niagara|SetNiagaraSystemAsset _bombsparkfx 0 true))))))

(event Custom|SetCollisionForcedDuringGame
  (params (NewForcedValue :type Enum_ForcedValueDuringGame%20Enum :default "NewEnumerator0"))
  ;[0.25 0.1274 0.02404] ForcedValuesDuringGame
  (if (Utilities|Enum|NotEqual(Enum) (Variables|Replicated|GetCollisionForcedDuringGame) NewForcedValue)
    (Variables|Replicated|SetCollisionForcedDuringGame NewForcedValue)
    (Networking|FlushNetDormancy)))

(event Custom|Async%20Load%20Visualization%20Material
  (bind _knob_rangevisualization (Variables|UserKnobs|GetKnob_RangeVisualization))
  (bind _rangevisualizationmaterials (Variables|Default|GetRangeVisualizationMaterials))
  (bind (_value _returnvalue) (Utilities|Map|Find _rangevisualizationmaterials _knob_rangevisualization))
  ;[1 0.7441 0 0.5] Load the Material asyncronously, then show the range visualization if the pawn is currently overlapping trigger
  (Variables|Default|SetMaterial%20Loaded false)
  (if _returnvalue
    (bind _loaded_object (Creative|Loading|AsyncLoadEditorSafe _value)
      (:Loaded
        (bind _asmaterial_interface (Utilities|Casting|CastToMaterialInterface _loaded_object))
        (Rendering|Material|SetMaterial (Variables|Default|GetRangeSphere) 0 _asmaterial_interface)
        (Variables|Default|SetMaterial%20Loaded true)
        (if (Editor|InEditor)
          (CallFunction|ShowRangeVisualization)
          (else
            ;;[1 0.4717 0] If we're currently overlapping a local pawn, it won't have started the range visualization as the material wasn't loaded. Start it now.
            (bind _array_element (Utilities|Array|ForEachLoopwithBreak (Fortnite|GetLocalFortPlayerControllers true) :!label _label)
              (:LoopBody
                (Utilities|IsValid (Pawn|GetControlledPawn _array_element)
                  (:Is%20Valid
                    (if (Collision|IsOverlappingActor (Variables|Default|GetRange%20Visualization%20Trigger) (Pawn|GetControlledPawn _array_element) :!class /Script/Engine.PrimitiveComponent)
                      (CallFunction|ShowRangeVisualization)
                      (goto _label Break))))))))))))

(event OnComponentBeginOverlap(RangeVisualizationTrigger)
  (params (OverlappedComponent :type Primitive%20Component%20Object%20Reference) (OtherActor :type Actor%20Object%20Reference) (OtherComp :type Primitive%20Component%20Object%20Reference) (OtherBodyIndex :type Integer) (bFromSweep :type Boolean) (SweepResult :type Hit%20Result%20Structure%20%28by%20ref%29))
  ;[1 0.8995 0 0.25] Show Visualization of Range when Near
  (if (Variables|Default|GetMaterial%20Loaded)
    (CallFunction|ShowRangeVisualization)))

(event OnComponentEndOverlap(RangeVisualizationTrigger)
  (params (OverlappedComponent :type Primitive%20Component%20Object%20Reference) (OtherActor :type Actor%20Object%20Reference) (OtherComp :type Primitive%20Component%20Object%20Reference) (OtherBodyIndex :type Integer))
  ;[1 0.8995 0 0.25] Show Visualization of Range when Near
  (CallFunction|HideRangeVisualization))

(event Custom|ShowRangeVisualization
  (bind _rangesphere (Variables|Default|GetRangeSphere))
  (bind _rangesphere_2 (Variables|Default|GetRangeSphere))
  ;[1 0.8995 0 0.25] Show Visualization of Range when Near
  (Networking|CanExecuteCosmeticEvents
    (:True
      (Development|SetHiddeninGame _rangesphere false false)
      (Rendering|SetVisibility _rangesphere true false)
      (if (Editor|InEditor)
        (Transformation|SetWorldScale3D (Variables|Default|GetRangeSphere) (CallFunction|GetRadiusVisualizationRange))
        (else
          (bind (_direction _lerp) (Timeline|Visualization%20Range%20%25 Play)
            (:Update
              (Rendering|Material|SetScalarParameterValue (Material|GetMID (Variables|Default|GetRangeSphere)) "Opacity" _lerp :!class /Script/Engine.MaterialInstanceDynamic)
              (Transformation|SetWorldScale3D (Variables|Default|GetRangeSphere) (Math|Vector|Lerp(Vector) "0.010000,0.010000,0.010000" (CallFunction|GetRadiusVisualizationRange) _lerp)))
            (:Finished
              (switch Utilities|FlowControl|Switch|SwitchonETimelineDirection _direction
                (:Backward
                  (Rendering|SetVisibility _rangesphere_2 false false)
                  (Development|SetHiddeninGame _rangesphere_2 true false) ; Using Set Hidden in Game because Set visiblity doesn't work in game.
                )))))))))

(event Custom|HideRangeVisualization
  (bind _rangesphere (Variables|Default|GetRangeSphere))
  ;[1 0.8995 0 0.25] Show Visualization of Range when Near
  (Networking|CanExecuteCosmeticEvents
    (:True
      (if (not (Editor|InEditor))
        (Timeline|Visualization%20Range%20%25 Reverse)
        (else
          (Transformation|SetWorldScale3D (Variables|Default|GetRangeSphere) "0.010000,0.010000,0.010000")
          (Rendering|SetVisibility _rangesphere false false)
          (Development|SetHiddeninGame _rangesphere true false) ; Using Set Hidden in Game because Set visiblity doesn't work in game.
        )))))

(event Custom|ToggleFuseFX
  (params (Activate :type Boolean))
  (if (== (Variables|UserOptions|Advanced|GetDeviceMesh) (Variables|Default|GetBombMeshReference))
    (Components|Activation|SetActive (Variables|Default|GetBombSparkFX) Activate false)
    (bind _loaded_object (Creative|Loading|AsyncLoadEditorSafe (select Activate "/Game/Creative/Devices/Bomb/Materials/CP_MI_Device_Bomb_Lit_01.CP_MI_Device_Bomb_Lit_01" "/Game/Creative/Devices/Bomb/Materials/CP_MI_Device_Bomb_01.CP_MI_Device_Bomb_01"))
      (:Loaded
        (bind _asmaterial_instance (Utilities|Casting|CastToMaterialInstance _loaded_object))
        (Rendering|Material|SetMaterial (Variables|Default|GetSM_MilitaryBarrel) 0 _asmaterial_instance)))))

(fn BlowUpBarrel
  (params (ActorToHit :type Actor%20Object%20Reference))
  ;[1 1 1] Send Gameplay event to ability to get explosion to fire and damage specified actor via OptionalObject
  (if (Variables|Actor|GetbCanBeDamaged ActorToHit)
    (Ability|SendGameplayEventtoActor self (Variables|Default|GetEventExplosionTag) (Utilities|Struct|MakeGameplayEventData (Variables|Default|GetEventExplosionTag) (select (Utilities|IsValid (Variables|Default|GetTriggerBarrelKiller) :!class /Script/Engine.KismetSystemLibrary) (Variables|Default|GetTriggerBarrelKiller) (Variables|Default|GetBarrelKiller)) 0 ActorToHit 0 "()" "(GameplayTags=)" "(GameplayTags=)" (Variables|UserKnobs|GetPlayerDamageSetting)))))

(fn GetRadiusVisualizationRange
  (return (Math|Conversions|ToVector(Float) (/ (Utilities|Select :Index (Variables|UserKnobs|GetKnob_RangeVisualization) :NewEnumerator0 0.0 :NewEnumerator1 (Variables|UserKnobs|GetDamageRangeSetting) :NewEnumerator2 (Math|Conversions|ToFloat(Integer) (Variables|UserKnobs|GetProximityDistanceSetting))) 5e+01))))

(fn GetCollisionStaticMesh
  (return "0, 0, 0" (Variables|/Script/Engine.StaticMeshComponent|GetStaticMesh (Variables|Default|GetSM_MilitaryBarrel))))

(fn GetMeshComponents
  (return 0))

(fn IgnoreWorldGridScaleOverride
  (return false))

(fn IsInPlayMode
  (bind _fortminigamelogic (Variables|Default|GetFortMinigameLogic)
    (:then
      (bind _currentminigame (Variables|/Script/FortniteGame.FortMinigameLogicComponent|GetCurrentMinigame _fortminigamelogic)
        (:then
          (return (Minigame|IsinPlayMode _currentminigame :!class /Script/FortniteGame.FortMinigame)))
        (:else
          _)))
    (:else
      _))
  (return false))

(fn IsTargetableByLinearMover
  (return true))

(fn LocalPlayerShouldSeeBarrel
  (Networking|SwitchHasAuthority
    (:Authority
      (return true))
    (:Remote
      (if (Math|Boolean|ORBoolean (Knobs|Team|IsTeamFriendlies (Variables|Default|GetShowTeamIndex)) (Knobs|Team|IsTeamHostiles (Variables|Default|GetShowTeamIndex)) (Knobs|Team|IsTeamNeutral (Variables|Default|GetShowTeamIndex)))
        (LocalVariables|SetLocalTeamAffiliation (Team|GetTeamAffiliation self (Game|GetPlayerPawn 0)))
        (if (and (Knobs|Team|IsTeamFriendlies (Variables|Default|GetShowTeamIndex)) (Utilities|Enum|Equal(Enum) (LocalVariables|GetLocalTeamAffiliation) "Friendly"))
          _
          (elif (and (Knobs|Team|IsTeamHostiles (Variables|Default|GetShowTeamIndex)) (Utilities|Enum|Equal(Enum) (LocalVariables|GetLocalTeamAffiliation) "Hostile"))
            _
            (elif (and (Knobs|Team|IsTeamNeutral (Variables|Default|GetShowTeamIndex)) (Utilities|Enum|Equal(Enum) (LocalVariables|GetLocalTeamAffiliation) "Neutral"))
              _
              (else
                (return false)))))
        (return true)
        (else
          (return (or (Knobs|Team|IsTeamAny (Variables|Default|GetShowTeamIndex)) (and (Knobs|Team|IsValidTeamIndex (Variables|Default|GetShowTeamIndex)) (== (Knobs|Team|GetTeamValueasIndex (Variables|Default|GetShowTeamIndex)) (Math|Conversions|ToInteger(Byte) (Team|GetActorPvPTeamIndex (Game|GetPlayerController 0))))))))))))

(fn MinigameIsInProgress
  (params (IgnoreWarmUp :type Boolean))
  (bind _self self)
  (bind _returnvalue (Minigame|GetMinigame _self :!class /Script/FortniteGame.MinigameBlueprintLibrary))
  (Utilities|IsValid _returnvalue
    (:Is%20Valid
      ;[1 1 1] Is in progress pure function includes Warmup time, optionally ignore that and only use "In Progress" time
      (if IgnoreWarmUp
        (return (Utilities|Enum|Equal(Enum) (Variables|/Script/FortniteGame.FortMinigame|GetCurrentState _returnvalue) "InProgress"))
        (else
          (return (Minigame|IsinProgress _returnvalue)))))
    (:Is%20Not%20Valid
      (return false))))

(fn OnExplodeFX
  (params (Location :type Vector))
  (bind _returnvalue (Transformation|GetActorLocation))
  (if (Math|Conversions|ToBoolean(Integer) (Variables|Default|GetPlayAudioKnob))
    (Audio|PlaySoundatLocation "/Game/Sounds/Creative/Gadgets/ExplodingBarrel/ExplodingBarrel_Explode_01_Cue.ExplodingBarrel_Explode_01_Cue" _returnvalue))
  (Effects|SpawnEmitteratLocation :EmitterTemplate "/Game/Creative/Devices/ExplosiveBarrel/P_ExplosiveBarrel_ExplosionMain.P_ExplosiveBarrel_ExplosionMain" :Location _returnvalue :Scale (Math|Vector|MakeVector (/ (Variables|UserKnobs|GetDamageRangeSetting) 512.0) (/ (Variables|UserKnobs|GetDamageRangeSetting) 512.0) (/ (Variables|UserKnobs|GetDamageRangeSetting) 512.0)) :bAutoDestroy true :PoolingMethod "None" :bAutoActivateSystem true))

(fn OnProximityTriggerableTeam
  (params (Actor :type Actor%20Object%20Reference))
  ;[1 1 1] Check if actor is on a team that should activate proximity trigger
  (return (or (Knobs|Team|IsTeamNone (Variables|UserKnobs|GetIgnoreTeamProximitySetting)) (not (== (Teams|TeamtoIndex (Team|GetActorTeam Actor)) (Knobs|Team|GetTeamValueasIndex (Variables|UserKnobs|GetIgnoreTeamProximitySetting)))))))

(fn OnRep_CollisionForcedDuringGame
  (CallFunction|UpdateBarrelMeshCollision))

(fn OnRep_ExplosiveFuseLit
  (Networking|CanExecuteCosmeticEvents
    (:True
      (CallFunction|ToggleFuseFX (Variables|Default|GetExplosiveFuseLit)))))

(fn RestoreToMaxHealth
  ;[1 0.4407 0.1563] Restores the building health to it's initial value
  (Networking|SwitchHasAuthority
    (:Authority
      (Health|SetHealth (Health|GetMaxHealth))
      (Variables|Replicated|SetHasExploded false)
      (CallFunction|SetMaxHealth)
      (Networking|FlushNetDormancy))))

(fn SetInvulneravilityByMinigameState
  (Networking|SwitchHasAuthority
    (:Authority
      (bind _isrunning (CallFunction|MinigameIsinProgress true))
      (Fortnite|SetCanbeDamaged self (and _isrunning (> (Variables|UserKnobs|GetStartingHealthSetting) 0.0))))))

(fn SetMaxHealth
  (bind _self self)
  (bind _returnvalue (Ability|GetAbilitySystemComponent _self))
  (bind _self_2 self)
  (bind _returnvalue_2 (Ability|GetAbilitySystemComponent _self_2))
  ;[1 1 1] Runs a gameplay effect on the actor that multiplies the max health by the setting value.
  (Networking|SwitchHasAuthority
    (:Authority
      (Utilities|IsValid _returnvalue_2
        (:Is%20Valid
          (bind _returnvalue_3 (Ability|GameplayEffect|AssignTagSetbyCallerMagnitude (GameplayEffects|MakeOutgoingSpec :self _returnvalue_2 :GameplayEffectClass (Variables|Default|GetModifyMaxHealth) :Level 0.0) (LocalVariables|GetSetMaxHealthDataTag) (Utilities|GetStartingHealth)))
          (Utilities|IsValid _returnvalue
            (:Is%20Valid
              (GameplayEffects|RemoveActiveGameplayEffect _returnvalue (Variables|Default|GetMaxHealthEffectHandler) -1)
              (bind _returnvalue_4 (GameplayEffects|ApplyGameplayEffectSpecToSelf _returnvalue _returnvalue_3))
              (Variables|Default|SetMaxHealthEffectHandler _returnvalue_4))))))))

(fn SetVisibilityForcedDuringGame
  (params (NewVisibilityForcedDuringGame :type Enum_ForcedValueDuringGame%20Enum :default "NewEnumerator0"))
  (Networking|SwitchHasAuthority
    (:Authority
      (if (not (and (Variables|Replicated|GetHasExploded) (Utilities|Enum|Equal(Enum) (Variables|GetNewVisibilityForcedDuringGame) "NewEnumerator1")))
        (if (Utilities|Enum|Equal(Enum) (Variables|UserKnobs|GetCollisionDuringGamesKnob) "NewEnumerator3")
          (CallFunction|SetCollisionForcedDuringGame (Variables|GetNewVisibilityForcedDuringGame)))
        (if (Utilities|Enum|Equal(Enum) (Variables|GetNewVisibilityForcedDuringGame) "NewEnumerator1")
          (Visibility|SetForcedVisibilityDuringGame (Variables|Default|GetVisibleInGameComponent) (select (CallFunction|LocalPlayerShouldSeeBarrel) "NewEnumerator1" "NewEnumerator2"))
          (else
            (Visibility|SetForcedVisibilityDuringGame (Variables|Default|GetVisibleInGameComponent) (Variables|GetNewVisibilityForcedDuringGame))))))))

(fn ShouldCancelExplosion
  (bind _isrunning (CallFunction|MinigameIsinProgress false))
  (return (not _isrunning)))

(fn ShouldInvisiblePropsBeAllowed
  (return false))

(fn ShouldTriggerProximityExplosion
  (params (OverlappingActor :type Object%20Reference) (OverlappingComponent :type Primitive%20Component%20Object%20Reference))
  ;[1 1 1] Ensures that overlapping actor should trigger the barrel.  Currently handles players and vehicles.
  (bind _asplayer_pawn_athena (Utilities|Casting|CastToPlayerPawn_Athena OverlappingActor)
    (:then
      (bind _returnvalue (CallFunction|OnProximityTriggerableTeam _asplayer_pawn_athena))
      (return (and _returnvalue (not (Vehicle|IsinVehicle _asplayer_pawn_athena)))))
    (:CastFailed
      (bind _asfort_vehicle_interface (Utilities|Casting|CastToFortVehicleInterface OverlappingActor)
        (:then
          (LocalVariables|SetlocalVehiclePassengers (Vehicle|GetAllPassengers _asfort_vehicle_interface))
          (if (Vehicle|HasDriver _asfort_vehicle_interface)
            (bind _returnvalue_2 (CallFunction|OnProximityTriggerableTeam (Vehicle|GetDriver _asfort_vehicle_interface)))
            (return _returnvalue_2)
            (elif (> (Utilities|Array|Length (LocalVariables|GetlocalVehiclePassengers)) 0)
              ;;[1 1 1] Checks if team of each passenger should trigger the barrel.  Breaks as soon as it finds one that should
              (bind _array_element (Utilities|Array|ForEachLoopwithBreak (LocalVariables|GetlocalVehiclePassengers) :!label _label)
                (:LoopBody
                  (bind _returnvalue_3 (CallFunction|OnProximityTriggerableTeam _array_element))
                  (bind _output_get (LocalVariables|SetAPassengerIsOnTriggerableTeam _returnvalue_3))
                  (if _output_get
                    (goto _label Break)))
                (:Completed
                  (return (LocalVariables|GetAPassengerIsOnTriggerableTeam))))
              (else
                (return false)))))
        (:CastFailed
          (return false))))))

(fn UpdateHealthWidget
  (bind _ignoreteamproximitysetting (Variables|UserKnobs|GetIgnoreTeamProximitySetting))
  (Networking|CanExecuteCosmeticEvents
    (:True
      (if (Math|Boolean|ANDBoolean (Variables|UserKnobs|GetbShowHealthBar) (> (Variables|UserKnobs|GetHealthBarIndex) 0) (!= (Variables|UserKnobs|GetStartingHealthSetting) -1.0) (or (Utilities|Name|NotEqual(Name) (Collision|GetCollisionProfileName (Variables|Default|GetSM_MilitaryBarrel)) (CreativeDevices|GetNamefromCollisionProfile "(Name=\"FortCreativeOnlyVolumeManager\")")) (Variables|/Script/Engine.SceneComponent|GetbVisible (Variables|Default|GetSM_MilitaryBarrel))))
        (LocalVariables|Setlocal%20Team%20Array (Knobs|Team|GetTeamArray self))
        (CallFunction|SetSuppressHealthBar true)
        (if (Knobs|Team|IsValidTeamIndex _ignoreteamproximitysetting)
          (Utilities|Array|RemoveItem (LocalVariables|Getlocal%20Team%20Array) (Knobs|Team|GetTeamValueasIndex _ignoreteamproximitysetting)) ; Removes the team from the hostile team array if they're not going to proximity activate it.
        )
        (Class|CreativeValueBarComponent|SpawnOrUpdateValueBar (Variables|Default|GetFloatingHealthBar) (Variables|Default|GetSM_MilitaryBarrel) self "/Game/Creative/Devices/Common/Components/Creative_FloatingHealthBar.Creative_FloatingHealthBar_C" 0 (Variables|UserKnobs|GetHealthBarDistance) 0 (Utilities|GetStartingHealth) 2e+01 (Variables|UserKnobs|GetbHealthBarRequireLOS) (select (Knobs|Team|IsValidTeamIndex (Variables|UserKnobs|GetIgnoreTeamSetting)) (Utilities|Array|MakeArray (Knobs|Team|GetTeamValueasIndex (Variables|UserKnobs|GetIgnoreTeamSetting))) 0) (LocalVariables|Getlocal%20Team%20Array) (Variables|Default|GetShowTeamIndex) (== (Variables|UserKnobs|GetHealthBarIndex) 2) (Variables|UserKnobs|GetHealthBarOverrideTextFriendly) (Variables|UserKnobs|GetHealthBarOverrideTextNeutral) (Variables|UserKnobs|GetHealthBarOverrideTextEnemy) (Variables|UserKnobs|GetHealthBarTextSize) true (Math|Conversions|ToBoolean(Integer) (Variables|UserKnobs|GetDisplayDistanceText)) (Math|Conversions|ToBoolean(Integer) (Variables|UserKnobs|GetClampToScreen)) (Math|Conversions|ToBoolean(Integer) (Variables|UserKnobs|GetShowOffscreenArrow)) (Creative|UI|IconPicker|GetIconLibraryTexture (Variables|UserKnobs|GetHealthBarIdentifierIndex)) true "(X=1.000000,Y=1.000000)" false false)
        (else
          (CallFunction|SetSuppressHealthBar (not (or (Variables|UserKnobs|GetbShowHealthBar) (== (Variables|UserKnobs|GetStartingHealthSetting) -1.0))))
          (Class|CreativeValueBarComponent|HideValueBar (Variables|Default|GetFloatingHealthBar)))))
    (:False
      (bind _floatinghealthbar (Variables|Default|GetFloatingHealthBar)
        (:then
          (Components|DestroyComponent _floatinghealthbar))
        (:else)))))

(fn UpdateVisualizationState
  (Networking|CanExecuteCosmeticEvents
    (:True
      (if (and (not (Class|CreativeCommonDeviceFunctionLibrary|IsinPlayMode self)) (Utilities|Enum|NotEqual(Enum) (Variables|UserKnobs|GetKnob_RangeVisualization) "NewEnumerator0"))
        (Collision|SetCollisionEnabled (Variables|Default|GetRange%20Visualization%20Trigger) "QueryOnly")
        (Transformation|SetWorldScale3D (Variables|Default|GetRangeSphere) (CallFunction|GetRadiusVisualizationRange))
        (CallFunction|AsyncLoadVisualizationMaterial)
        (else
          _)))
    (:False
      _))
  (Collision|SetCollisionEnabled (Variables|Default|GetRange%20Visualization%20Trigger) "NoCollision")
  (CallFunction|HideRangeVisualization))

(fn UpdateBarrelMeshCollision
  (bind _collisionprofilename (LocalVariables|GetCollisionProfileName))
  (Utilities|FlowControl|Sequence
    (:then_0
      (LocalVariables|SetNoCollisionProfileName (CreativeDevices|GetNamefromCollisionProfile "(Name=\"FortCreativeOnlyVolumeManager\")"))
      (LocalVariables|SetEditModeCollisionProfileName (CreativeDevices|GetNamefromCollisionProfile "(Name=\"FortPhysicsBodyMeshPhysics\")"))
      (LocalVariables|SetPlayModeProfileName (CreativeDevices|GetNamefromCollisionProfile "(Name=\"FortPhysicsBodyMeshPhysics\")"))
      (if (CallFunction|IsinPlayMode)
        (switch Utilities|FlowControl|Switch|SwitchonEnum_ForcedValueDuringGame (Variables|Replicated|GetCollisionForcedDuringGame)
          (:NewEnumerator0
            (switch Utilities|FlowControl|Switch|SwitchonEnum_CollisionDuringGamesSetting (Variables|UserKnobs|GetCollisionDuringGamesKnob)
              (:NewEnumerator0
                (LocalVariables|SetCollisionProfileName (LocalVariables|GetNoCollisionProfileName)))
              (:NewEnumerator1
                (LocalVariables|SetCollisionProfileName (LocalVariables|GetPlayModeProfileName)))
              (:NewEnumerator3
                (if (Variables|UserKnobs|GetbIsVisibleDuringGamesKnob)
                  (LocalVariables|SetCollisionProfileName (LocalVariables|GetPlayModeProfileName))
                  (else
                    (LocalVariables|SetCollisionProfileName (LocalVariables|GetNoCollisionProfileName)))))))
          (:NewEnumerator1
            (LocalVariables|SetCollisionProfileName (LocalVariables|GetPlayModeProfileName)))
          (:NewEnumerator2
            (LocalVariables|SetCollisionProfileName (LocalVariables|GetNoCollisionProfileName))))
        (else
          (LocalVariables|SetCollisionProfileName (LocalVariables|GetEditModeCollisionProfileName)))))
    (:then_1
      (if (Utilities|Name|NotEqual(Name) _collisionprofilename "None")
        (Collision|SetCollisionProfileName (Variables|Default|GetSM_MilitaryBarrel) _collisionprofilename true :!class /Script/Engine.PrimitiveComponent)
        (CallFunction|UpdateHealthWidget)))))

(fn UpdateVisibilityComponent
  (Visibility|SetVisiblityDuringPhase (Variables|Default|GetVisibleInGameComponent) (select (and (Variables|UserKnobs|GetbIsVisibleDuringGamesKnob) (CallFunction|LocalPlayerShouldSeeBarrel)) "NewEnumerator3" "NewEnumerator1")))

(fn ConstructionScript
  (|CallParentFunction)
  (Variables|/Script/Engine.SceneComponent|SetbAbsoluteScale true (Variables|Default|GetRangeSphere))
  (Variables|/Script/Engine.SceneComponent|SetbAbsoluteScale true (Variables|Default|GetProximityTrigger)))
